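# Post-process AnimBP for the M_MED_IslandDepth_Head player head mesh that drives its custom neck and head bones.
# The AnimGraph chains three full-weight Constraint nodes in component space, tying custom_neck_01, custom_neck_02 and custom_head to neck_01, neck_02 and head.
# Each update, the EventGraph stores inverted, scaled neck rotation deltas from ref pose relative to spine_05.
# The AnimGraph does not read these stored rotators.
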AnimBlueprint /BRCosmetics/Characters/Player/Male/Medium/Heads/M_MED_IslandDepth_Head/Meshes/M_MED_IslandDepth_Head_AnimBP_PostProcess
target_skeleton=&/M_MED_IslandDepth_Head_Skeleton
graphs [
 AnimGraph {nodes [
   n1 = "Constraint - custom_neck_01 (P)"(ComponentPose=Local To Component(LocalPose=Input Pose()), Alpha=1, ConstraintWeights_0=1) {boneToModify {boneName=custom_neck_01}, constraintSetup [{targetBone {boneName=neck_01}}], constraintWeights [1]}
   n2 = "Constraint - custom_neck_02 (P)"(ComponentPose=@n1, Alpha=1, ConstraintWeights_0=1) {boneToModify {boneName=custom_neck_02}, constraintSetup [{targetBone {boneName=neck_02}}], constraintWeights [1]}
   n3 = "Constraint - custom_head (P)"(ComponentPose=@n2, Alpha=1, ConstraintWeights_0=1) {boneToModify {boneName=custom_head}, constraintSetup [{targetBone {boneName=head}}], constraintWeights [1]}
   Output Pose(Result=Component To Local(ComponentPose=@n3))
  ]}
 EventGraph {nodes [
   Event BlueprintUpdateAnimation()
   n1 = GetOwnerSkeletalMeshComponent()
   n2 = BreakRotator(InRot=BreakTransform(InTransform=GetDeltaTransformFromRefPose(self=@n1, BoneName=neck_02, BaseName=spine_05)).Rotation)
   > Set neck02 rotate(neck02 rotate:Rotator=MakeRotator(
     Roll=float * float!(A=@n2, B=-1, C:double=0.6) -> (ReturnValue)
     Pitch=float * float(A=@n2.Pitch, B=-1, C=0.6)
     Yaw=float * float(A=@n2.Yaw, B=-1, C=0.6)
    )) -> (Output_Get:Rotator)
   n3 = BreakRotator(InRot=BreakTransform(InTransform=GetDeltaTransformFromRefPose(self=@n1, BoneName=neck_01, BaseName=spine_05)).Rotation)
   > Set neck01 rotate(neck01 rotate:Rotator=MakeRotator(Roll=float * float(A=@n3, B=-1, C=0.6), Pitch=float * float(A=@n3.Pitch, B=-1, C=0.6), Yaw=float * float(A=@n3.Yaw, B=-1, C=0.6))) -> (Output_Get:Rotator)
  ]}
]
variables [
 neck01 rotate {type=Rotator}
 neck02 rotate {type=Rotator}
]
